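# Background world material in the CattusDoggus event's WaterSplash folder, shading distant ocean that transitions into beach and land.
# World-height, slope, vertex-color and texture masks blend water, sand, grass, city and rock, though the UseOcean switch defaults to ocean only.
# Layered panning world-space normals flatten with camera distance, and High quality adds foam.
# Skybox parameter-collection time-of-day colors drive distance-masked fake-fog tint and emissive.
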Material /Figment_S09/Effects/Fort_Effects/Events/CattusDoggus/Doggus/WaterSplash/M_Backround_WatertoLand
graph {
 properties {
  bUseMaterialAttributes=true
  bUsedWithInstancedStaticMeshes=true
  bUsedWithMeshParticles=true
  bUsedWithNanite=true
  pixelDepthOffsetMode=PDOM_Legacy
  usedShadingModels=Default Lit
 }
 nodes [
  Constant3Vector() {constant=<rgba 0 0 1 0>}
  Constant() {r=0.1}
  Constant()
  Constant()
  n1 = NamedRerouteUsage()
  n2 = ScalarParameter() {
   defaultValue=500
   desc="At low END OCean overcomes landmasses, this WPO the large ocean sections below the Far Water"
   parameterName=MobileOceanWPO
  }
  n3 = Multiply(A=Multiply(A=OneMinus(None=@n1), B=Constant3Vector() {constant=<rgba 0 0 -1 0>}), B=@n2) {constB=500}
  FeatureLevelSwitch(ES2=@n3, ES3_1=@n3)
  n4 = WorldPosition() {worldPositionShaderOffset=WPT_ExcludeAllShaderOffsets}
  n5 = ComponentMask(None=@n4.XYZ) {b=true}
  n6 = ScalarParameter() {defaultValue=4335.1958, parameterName=TransitionBeach_Offset}
  n7 = Clamp(None=Divide(A=Add(A=@n5, B=@n6), B=ScalarParameter() {defaultValue=779.4005, parameterName=TransitionBeach_Distance}))
  LinearInterpolate(A=Constant(), Alpha=@n7)
  LinearInterpolate(A=Constant() {r=1}, B=Constant() {r=0.85})
  MakeMaterialAttributes!(
   BaseColor=NamedRerouteUsage()
   Metallic
   Specular=Constant() {r=0.1}
   Roughness=Constant() {r=0.2}
   Anisotropy
   EmissiveColor=NamedRerouteUsage()
   Opacity
   OpacityMask
   Normal=NamedRerouteUsage()
   Tangent
   WorldPositionOffset=Constant()
   SubsurfaceColor
   ClearCoat
   ClearCoatRoughness
   AmbientOcclusion
   Refraction
   CustomizedUVs_0
   CustomizedUVs_1
   CustomizedUVs_2
   CustomizedUVs_3
   CustomizedUVs_4
   CustomizedUVs_5
   CustomizedUVs_6
   CustomizedUVs_7
   PixelDepthOffset
   ShadingModel
   Displacement
  )
  MaterialFunctionCall() {materialFunction=&FlattenNormal}
  n8 = WorldPosition() {worldPositionShaderOffset=WPT_ExcludeAllShaderOffsets}
  n9 = ComponentMask(None=@n8.XYZ) {g=true, r=true}
  n10 = AppendVector(A=ScalarParameter() {defaultValue=280645.75, parameterName=UVSize_1}, B=ScalarParameter() {defaultValue=346358.5312, parameterName=UVSize_3})
  n11 = TextureSample(UVs=VertexInterpolator(VS=Divide(A=@n9, B=@n10)).PS) {samplerType=SAMPLERTYPE_Normal, texture=/Game/Packages/Fortnite_UniqueMaterials_1/Textures/T_Water_TilingNormal_Waves_02}
  n12 = MaterialFunctionCall(Normal=@n11.RGB, Flatness=ScalarParameter() {defaultValue=0.3, parameterName=NormalFlatten_large_1}) {materialFunction=&FlattenNormal}
  n13 = ComponentMask(None=@n8.XYZ) {g=true, r=true}
  n14 = ScalarParameter() {defaultValue=10575.0576, parameterName=UVSize}
  n15 = Divide(A=@n13, B=ScalarParameter() {defaultValue=265404.75, parameterName=WaveSize})
  n16 = Add(
   A=VertexInterpolator(VS=Divide(A=@n13, B=@n14)).PS
   B=Multiply(A=Sine(None=@n15), B=ScalarParameter() {defaultValue=2.3193, parameterName=Distort})
  )
  n17 = Add(A=Multiply(A=@n16) {constB=0.8556}) {constB=0.3215}
  n18 = Multiply(A=Time(), B=ScalarParameter() {defaultValue=0.4519, parameterName=speed})
  n19 = Add(A=@n16) {constB=0.1255}
  n20 = ComponentMask(None=TextureSample(UVs=Panner(Coordinate=@n19, Time=@n18) {bFractionalPart=true, speedX=-0.1208, speedY=0.1278}) {samplerType=SAMPLERTYPE_Normal, texture=&T_Water_TilingNormal_Waves_02}.RGB) {g=true, r=true}
  n21 = TextureSample(UVs=Add(A=Panner(Coordinate=@n17, Time=@n18) {bFractionalPart=true, speedX=-0.0625, speedY=-0.0417}, B=Multiply(A=@n20) {constB=0.035})) {samplerType=SAMPLERTYPE_Normal, texture=&T_Water_TilingNormal_Waves_02}
  n22 = Multiply(A=ComponentMask(None=@n21.RGB) {g=true, r=true}, B=ScalarParameter() {defaultValue=0.1619, group=Deep Water, parameterName=Noise 1}) {constB=0.04}
  n23 = Panner(Coordinate=@n16, Time=@n18) {bFractionalPart=true, speedX=-0.025, speedY=-0.0675}
  n24 = TextureSample(UVs=Add(A=@n22, B=Add(A=@n23) {constB=0.333})) {samplerType=SAMPLERTYPE_Normal, texture=&T_Water_TilingNormal_Waves_02}
  n25 = Divide(A=@n13, B=ScalarParameter() {defaultValue=4696.0737, parameterName=UVSize_2})
  n26 = Add(A=@n22, B=VertexInterpolator(VS=Panner(Coordinate=@n25) {bFractionalPart=true, speedX=-0.06, speedY=-0.0167}).PS)
  n27 = MaterialFunctionCall(
   BaseNormal=Normalize(VectorInput=MaterialFunctionCall(BaseNormal=@n21.RGB, AdditionalNormal=@n24.RGB) {materialFunction=&BlendAngleCorrectedNormals}.Result)
   AdditionalNormal=Multiply(
    A=TextureSample(UVs=@n26) {samplerType=SAMPLERTYPE_Normal, texture=&T_Water_TilingNormal_Waves_02}.RGB
    B=Constant3Vector() {constant=<rgba 0.65 0.65 1 0>}
   )
  ) {materialFunction=&BlendAngleCorrectedNormals}
  n28 = ScalarParameter() {defaultValue=0.97, parameterName=NormalFlatten}
  n29 = ScalarParameter() {defaultValue=1, parameterName=NormalFlatten_2}
  n30 = MaterialFunctionCall(
   Fade Length=ScalarParameter() {defaultValue=287068.375, parameterName=FadeDistance}
   Fade Offset=ScalarParameter() {defaultValue=11003.7119, parameterName=FadeOffset}
  ) {materialFunction=&CameraDepthFade}
  MaterialFunctionCall(
   BaseNormal=MaterialFunctionCall(Normal=@n11.RGB, Flatness=ScalarParameter() {defaultValue=0.9766, parameterName=NormalFlatten_large}) {materialFunction=&FlattenNormal}.Result
   AdditionalNormal=MaterialFunctionCall(
    Normal=MaterialFunctionCall(BaseNormal=@n12.Result, AdditionalNormal=@n27.Result) {materialFunction=&BlendAngleCorrectedNormals}.Result
    Flatness=LinearInterpolate(A=@n28, B=@n29, Alpha=@n30.Result)
   ) {materialFunction=&FlattenNormal}.Result
  ) {materialFunction=&BlendAngleCorrectedNormals}
  n31 = TextureSample() {texture=/Game/Environments/World/Backgrounds/Locales/Temperate/Textures/T_Background_Buildings}
  n32 = ScalarParameter() {defaultValue=0.27, parameterName=SpecularBuildings}
  MaterialFunctionCall!(
   IsOcean=StaticBoolParameter() {defaultValue=true, parameterName=IsOcean}
   IsTerrain
   In=StaticSwitchParameter(True=MakeMaterialAttributes(BaseColor=@n31.RGB, Specular=@n32)) {parameterName=IsBuildingModeled}
  ) {materialFunction=/Game/Packages/Fortress_SharedMaterials/Base_Material_Functions/MF_HLODWPO}
  n33 = ScalarParameter() {defaultValue=148862.7188, parameterName=size}
  MaterialFunctionCall(
   TextureObject=TextureObject() {texture=/Game/Packages/Fortress_Rubble/Textures/T_CliffTile_D}
   TextureSize=AppendVector(A=AppendVector(A=@n33, B=@n33), B=@n33)
  ) {materialFunction=&WorldAlignedTexture}
  n34 = Divide(
   A=Add(A=@n5, B=ScalarParameter() {defaultValue=0.1661, parameterName=Height_Grass_Offset})
   B=ScalarParameter() {defaultValue=18036.5312, parameterName=Height_Grass_Distance}
  )
  n35 = VertexNormalWS()
  n36 = Clamp(None=LinearInterpolate(
    A=ScalarParameter() {defaultValue=20.9442, parameterName=CityTransiton_1}
    B=ScalarParameter() {defaultValue=-5.2032, parameterName=CityTransiton_2}
    Alpha=ComponentMask(None=@n35) {b=true}
   ))
  Multiply(A=Clamp(None=@n34), B=@n36)
  n37 = ScalarParameter() {defaultValue=5107.3623, parameterName=TransitionOcean_Offset}
  n38 = Clamp(None=Divide(A=Add(A=@n5, B=@n37), B=ScalarParameter() {defaultValue=567.3769, parameterName=TransitionOcean_Distance}))
  NamedRerouteDeclaration(None=@n38) {name=OceanBlendAlpha, nodeColor=<rgba 0.8281 0 1 1>}
  n39 = Constant3Vector() {constant=<rgba 0 0 1 0>}
  n40 = ComponentMask(None=Divide(A=WorldPosition() {worldPositionShaderOffset=WPT_ExcludeAllShaderOffsets}.XYZ, B=Constant() {r=18286.0156})) {g=true, r=true}
  n41 = Multiply(A=Constant() {r=0.119}, B=@n40)
  n42 = Multiply(A=Time(), B=Constant() {r=0.6})
  n43 = Panner(Coordinate=Multiply(A=Constant() {r=0.2557}, B=@n40), Time=@n42) {bFractionalPart=true, speedX=0.01, speedY=-0.005}
  n44 = MaterialFunctionCall(
   BaseNormal=TextureSample(UVs=Panner(Coordinate=@n41, Time=@n42) {bFractionalPart=true, speedX=-0.015, speedY=0.005}) {samplerType=SAMPLERTYPE_Normal, texture=&/T_Water2_N}.RGB
   AdditionalNormal=TextureSample(UVs=@n43) {samplerType=SAMPLERTYPE_Normal, texture=&/WaterFoam0001_19_N}.RGB
  ) {materialFunction=&BlendAngleCorrectedNormals}
  n45 = MaterialFunctionCall(Fade Length=Constant() {r=185195.2031}, Fade Offset=Constant()) {materialFunction=&CameraDepthFade}
  n46 = LinearInterpolate(A=Constant() {r=1}, B=Constant() {r=0.85}, Alpha=@n45.Result)
  n47 = Time()
  n48 = Multiply(A=@n40, B=Constant() {r=1})
  n49 = Add(
   A=TextureSample(UVs=Panner(Coordinate=@n40, Time=@n47) {bFractionalPart=true, speedX=0.009, speedY=-0.009}) {samplerType=SAMPLERTYPE_Normal, texture=&T_Water2_N}.RGB
   B=TextureSample(UVs=Panner(Coordinate=@n48, Time=@n47) {bFractionalPart=true, speedX=-0.003, speedY=0.006}) {samplerType=SAMPLERTYPE_Normal, texture=&WaterFoam0001_19_N}.RGB
  )
  n50 = LinearInterpolate(A=Constant() {r=0.2}, B=Constant() {r=1}, Alpha=@n45.Result)
  n51 = Fresnel(ExponentIn=Constant() {r=4}, Normal=VertexNormalWS()) {baseReflectFraction=0}
  n52 = MaterialFunctionCall(
   Normal=Add(
    A=MaterialFunctionCall(Normal=@n44.Result, Flatness=@n46) {materialFunction=&FlattenNormal}.Result
    B=MaterialFunctionCall(Normal=@n49, Flatness=@n50) {materialFunction=&FlattenNormal}.Result
   )
   Flatness=@n51
  ) {materialFunction=&FlattenNormal}
  n53 = MaterialFunctionCall(Normal=@n44.Result, Flatness=LinearInterpolate(A=Constant() {r=0.85}, B=Constant() {r=1}, Alpha=@n51)) {materialFunction=&FlattenNormal}
  n54 = QualitySwitch(Default=@n52.Result, Low=@n53.Result, High=@n52.Result, Medium=@n52.Result)
  NamedRerouteDeclaration(None=MaterialFunctionCall(
    BaseNormal=Transform(None=@n39) {transformSourceType=TRANSFORMSOURCE_World, transformType=TRANSFORM_Tangent}
    AdditionalNormal=ShadingPathSwitch(Default=@n54, Mobile=@n53.Result) {desc=Migrated from FeatureLevelSwitch}
   ) {materialFunction=&BlendAngleCorrectedNormals}.Result) {name=MA_Normal, nodeColor=<rgba 0.8281 1 0 1>}
  n55 = Add(A=NamedRerouteUsage(), B=ScalarParameter() {defaultValue=4509.2178, parameterName=Transition_Offset_2})
  n56 = ScalarParameter() {defaultValue=306871.9375, parameterName=Transition_Distance_1}
  n57 = LinearInterpolate(
   A=ScalarParameter() {defaultValue=0.3518, parameterName=FakeFogBrightness}
   B=ScalarParameter() {defaultValue=1, parameterName=FakeFogBrightness_2}
   Alpha=Clamp(None=Divide(A=@n55, B=@n56))
  )
  n58 = CollectionParameter() {
   collection=/Game/Environments/World/Backgrounds/Universal/MasterMaterials/MPC_Skybox_World
   parameterName="Morn_R=Distance_G=DiffuseAmt_B=EmissiveAmt"
  }
  n59 = CollectionParameter() {
   collection=&MPC_Skybox_World
   parameterName="Day_R=Distance_G=DiffuseAmt_B=EmissiveAmt"
  }
  n60 = CollectionParameter() {
   collection=&MPC_Skybox_World
   parameterName="Eve_R=Distance_G=DiffuseAmt_B=EmissiveAmt"
  }
  n61 = CollectionParameter() {
   collection=&MPC_Skybox_World
   parameterName="Night_R=Distance_G=DiffuseAmt_B=EmissiveAmt"
  }
  n62 = MaterialFunctionCall(Float3=MaterialFunctionCall(Color_Morning=@n58, Color_Day=@n59, Color_Evening=@n60, Color_Night=@n61) {materialFunction=&Color_Time_of_Day}.Result) {materialFunction=&BreakOutFloat3Components}
  n63 = ComponentMask(None=WorldPosition() {worldPositionShaderOffset=WPT_ExcludeAllShaderOffsets}.XYZ) {g=true, r=true}
  n64 = SphereMask(A=@n63, B=Constant2Vector(), Radius=@n62.R, Hardness=Constant())
  n65 = MaterialFunctionCall(
   Color_Morning=CollectionParameter() {collection=&MPC_Skybox_World, parameterName=MorningColor}
   Color_Day=CollectionParameter() {collection=&MPC_Skybox_World, parameterName=DayColor}
   Color_Evening=CollectionParameter() {collection=&MPC_Skybox_World, parameterName=EveningColor}
   Color_Night=CollectionParameter() {collection=&MPC_Skybox_World, parameterName=NightColor}
  ) {materialFunction=&Color_Time_of_Day}
  n66 = MaterialFunctionCall(
   Color_Morning=CollectionParameter() {collection=&MPC_Skybox_World, parameterName=MorningLitColor}
   Color_Day=CollectionParameter() {collection=&MPC_Skybox_World, parameterName=DayLitColor}
   Color_Evening=CollectionParameter() {collection=&MPC_Skybox_World, parameterName=EveningLitColor}
   Color_Night=CollectionParameter() {collection=&MPC_Skybox_World, parameterName=NightLitColor}
  ) {materialFunction=&Color_Time_of_Day}
  n67 = CollectionParameter() {
   collection=&MPC_Skybox_World
   parameterName="Morn_R=LitLength_G=LitRot_B=LitOffset"
  }
  n68 = CollectionParameter() {
   collection=&MPC_Skybox_World
   parameterName="Day_R=LitLength_G=LitRot_B=LitOffset"
  }
  n69 = CollectionParameter() {
   collection=&MPC_Skybox_World
   parameterName="Eve_R=LitLength_G=LitRot_B=LitOffset"
  }
  n70 = CollectionParameter() {
   collection=&MPC_Skybox_World
   parameterName="Night_R=LitLength_G=LitRot_B=LitOffset"
  }
  n71 = MaterialFunctionCall(Float3=MaterialFunctionCall(Color_Morning=@n67, Color_Day=@n68, Color_Evening=@n69, Color_Night=@n70) {materialFunction=&Color_Time_of_Day}.Result) {materialFunction=&BreakOutFloat3Components}
  n72 = Subtract(A=Add(A=@n63, B=Constant() {r=4000000}), B=AppendVector(A=@n71.G, B=@n71.G))
  n73 = Constant2Vector()
  n74 = Constant()
  n75 = Multiply(
   A=Multiply(A=@n62.B, B=OneMinus(None=@n64))
   B=LinearInterpolate(
    A=ComponentMask(None=@n65.Result) {b=true, g=true, r=true}
    B=ComponentMask(None=@n66.Result) {b=true, g=true, r=true}
    Alpha=SphereMask(A=@n72, B=@n73, Radius=@n71.R, Hardness=@n74)
   )
  ) {desc=Emissive}
  NamedRerouteDeclaration(None=VertexInterpolator(VS=Multiply(A=@n57, B=@n75)).PS) {name=MA_Emissive, nodeColor=<rgba 0.2891 0 1 1>}
  NamedRerouteDeclaration(None=@n5) {name=B_Mask, nodeColor=<rgba 0.9922 1 0 1>}
  n76 = NamedRerouteUsage()
  n77 = VertexInterpolator(VS=@n75)
  NamedRerouteDeclaration(None=LinearInterpolate(A=@n76, B=Multiply(A=@n76, B=@n77.PS), Alpha=@n62.G)) {name=MA_BaseColor, nodeColor=<rgba 0.5938 1 0 1>}
  n78 = ComponentMask(None=WorldPosition() {worldPositionShaderOffset=WPT_ExcludeAllShaderOffsets}.XYZ) {g=true, r=true}
  n79 = Constant2Vector() {g=-153000, r=-153000}
  n80 = TextureSample(UVs=Divide(A=Subtract(A=@n78, B=@n79), B=Constant() {r=306200})) {
   samplerSource=SSM_Clamp_WorldGroupSettings
   samplerType=SAMPLERTYPE_LinearColor
   texture=&/T_Athena_Terrain_Mask
  }
  n81 = Constant() {r=0.2}
  n82 = LinearInterpolate(
   A=VectorParameter() {parameterName=deepwater, defaultValue=<rgba 0.0049 0.0249 0.05 1>}.RGB
   B=VectorParameter() {parameterName=shallowwater, defaultValue=<rgba 0.0469 0.5 0.4635 1>}.RGB
   Alpha=Multiply(A=Clamp(None=@n80.A, Max=@n81) {clampMode=CMODE_ClampMin, maxDefault=0.5}, B=VertexColor().A)
  )
  n83 = ComponentMask(None=Divide(A=WorldPosition() {worldPositionShaderOffset=WPT_ExcludeAllShaderOffsets}.XYZ, B=Constant() {r=6202.6328})) {g=true, r=true}
  n84 = Time()
  n85 = TextureSample(UVs=Panner(Coordinate=@n83, Time=@n84) {bFractionalPart=true, speedX=-0.03, speedY=0.03}) {texture=&/Foam}
  n86 = TextureSample(UVs=Panner(Coordinate=Multiply(A=@n83) {constB=1.5}, Time=@n84) {bFractionalPart=true, speedX=0.02, speedY=-0.1}) {texture=&Foam}
  n87 = Multiply(A=Constant3Vector() {constant=<rgba 1.5469 2.5102 3 0>}, B=Clamp(None=Subtract(A=@n85.G, B=@n86.G)))
  n88 = Constant() {r=3}
  n89 = Clamp(None=Subtract(A=Multiply(A=@n80.A, B=@n88)) {constB=0.6})
  n90 = QualitySwitch(Default=@n82, Low=@n82, High=Add(A=@n82, B=Multiply(A=@n87, B=@n89)), Medium=@n82)
  n91 = ComponentMask(None=@n4.XYZ) {g=true, r=true}
  n92 = ScalarParameter() {defaultValue=478490.3125, parameterName=TextureSize_1}
  n93 = TextureSample(UVs=VertexInterpolator(VS=Divide(A=@n91, B=@n92)).PS) {samplerType=SAMPLERTYPE_Masks, texture=&/T_OceanGround_M}
  n94 = ScalarParameter() {defaultValue=0.2248, parameterName=SandBlend}
  n95 = ScalarParameter() {defaultValue=-0.3143, parameterName=SandBlend_1}
  n96 = LinearInterpolate(
   A=LinearInterpolate(
    A=VectorParameter() {parameterName=Shallow, defaultValue=<rgba 0.0306 0.1031 0.151 1>}.RGB
    B=ShadingPathSwitch(Default=@n90, Mobile=@n82) {desc=Migrated from FeatureLevelSwitch}
    Alpha=@n93.G
   )
   B=VectorParameter() {parameterName=Ground, defaultValue=<rgba 0.1389 0.3107 0.375 1>}.RGB
   Alpha=Clamp(None=LinearInterpolate(A=@n94, B=@n95, Alpha=@n93.G))
  )
  NamedRerouteDeclaration(None=@n96) {name=isOcean, nodeColor=<rgba 1 0 0.1172 1>}
  n97 = VectorParameter() {parameterName=Beach, defaultValue=<rgba 0.0781 0.0717 0.0553 1>}
  n98 = Multiply(
   A=LinearInterpolate(
    A=ScalarParameter() {defaultValue=2.4369, parameterName=SandBlend_2}
    B=ScalarParameter() {defaultValue=-3.7123, parameterName=SandBlend_3}
    Alpha=@n93.G
   )
   B=OneMinus(None=VertexColor().R)
  )
  n99 = Clamp(None=Add(A=Clamp(None=@n98), B=@n38))
  n100 = ScalarParameter() {defaultValue=81418.2188, parameterName=Height_ColorGradient_Offset}
  n101 = Clamp(None=Divide(A=Add(A=@n5, B=@n100), B=ScalarParameter() {defaultValue=114045.4453, parameterName=Height_ColorGradient_Distance}))
  n102 = ShadingPathSwitch(
   Default=LinearInterpolate(
    A=VectorParameter() {parameterName=ColorBottom, defaultValue=<rgba 1 1 1 1>}.RGB
    B=VectorParameter() {parameterName=ColorTop, defaultValue=<rgba 0.7344 0.4878 0.0379 1>}.RGB
    Alpha=VertexInterpolator(VS=@n101).PS
   )
   Mobile=VectorParameter() {parameterName=ColorMobile, defaultValue=<rgba 0.5312 0.5312 0.5312 1>}.RGB
  ) {desc=Migrated from FeatureLevelSwitch}
  n103 = VertexInterpolator(VS=Divide(A=ComponentMask(None=@n4.XYZ) {g=true, r=true}, B=ScalarParameter() {defaultValue=-361300.6562, parameterName=TextureSize_3}))
  n104 = VectorParameter() {parameterName=Grass, defaultValue=<rgba 0.5323 1 0.2442 1>}
  n105 = ScalarParameter() {defaultValue=13.1194, parameterName=TextureSize}
  n106 = Multiply(A=TextureCoordinate(), B=@n105)
  n107 = TextureSample(UVs=@n106) {texture=/Game/Environments/World/Backgrounds/Locales/Temperate/Textures/T_Skybox_TreeCanopy_03}
  n108 = VertexInterpolator(VS=Clamp(None=Add(A=@n34, B=@n36)))
  n109 = VectorParameter() {parameterName=Rock, defaultValue=<rgba 3 3 3 1>}
  n110 = TextureSample() {texture=/Game/Environments/Homebase/Textures/T_HB_Mountain_D}
  n111 = Multiply(A=TextureCoordinate(), B=ScalarParameter() {defaultValue=5.0987, parameterName=TilingRockSize})
  n112 = ScalarParameter() {defaultValue=1.3741, parameterName=RockTransition_Low}
  n113 = ScalarParameter() {defaultValue=0.1146, parameterName=RockTransition_High}
  n114 = TextureSampleParameter2D() {parameterName=Normal, samplerType=SAMPLERTYPE_Normal, texture=&FlatNormalBlue}
  n115 = ComponentMask(None=Transform(None=@n114.RGB)) {b=true}
  n116 = ScalarParameter() {parameterName=Height_Offset_1}
  n117 = Clamp(None=Divide(A=Add(A=@n5, B=@n116), B=ScalarParameter() {defaultValue=48540.4531, parameterName=Height_Distance_2}))
  n118 = Clamp(None=Multiply(A=LinearInterpolate(A=@n112, B=@n113, Alpha=@n115), B=VertexInterpolator(VS=@n117).PS))
  n119 = LinearInterpolate(
   A=LinearInterpolate(
    A=TextureSample(UVs=@n103.PS) {texture=/Game/Environments/World/Backgrounds/Locales/Temperate/Textures/T_BackgroundCivilisation_D}.RGB
    B=Multiply(A=@n104.RGB, B=@n107.RGB)
    Alpha=@n108.PS
   )
   B=Multiply(A=Multiply(A=@n109.RGB, B=@n110.RGB), B=TextureSample(UVs=@n111) {texture=&T_CliffTile_D}.RGB)
   Alpha=@n118
  )
  n120 = StaticSwitchParameter(True=@n96, False=LinearInterpolate(A=LinearInterpolate(A=@n96, B=@n97.RGB, Alpha=@n99), B=Multiply(A=@n102, B=@n119), Alpha=@n7)) {defaultValue=true, parameterName=UseOcean}
  n121 = VertexNormalWS()
  n122 = StaticSwitchParameter(
   True=LinearInterpolate(
    A=VectorParameter() {parameterName=BuildingColor, defaultValue=<rgba 0.25 0.2199 0.2057 1>}.RGB
    B=@n120
    Alpha=ComponentMask(None=@n121) {b=true}
   )
   False=@n120
  ) {parameterName=IsBuilding}
  Reroute(None=@n122)
  Rotator(Coordinate=@n43, Time=Constant() {r=0.8658})
  n123 = ScalarParameter() {defaultValue=1, parameterName=Normal_3}
  n124 = LinearInterpolate(
   A=ScalarParameter() {defaultValue=-4.3911, parameterName=Normal_1}
   B=ScalarParameter() {defaultValue=1, parameterName=Normal_2}
   Alpha=@n118
  )
  n125 = LinearInterpolate(
   A=MaterialFunctionCall(
    Normal=TextureSample(UVs=@n106) {samplerType=SAMPLERTYPE_Normal, texture=/Game/Environments/World/Backgrounds/Locales/Temperate/Textures/T_Skybox_TreeCanopy_01_N}.RGB
    Flatness=LinearInterpolate(A=@n123, B=@n124, Alpha=@n108.PS)
   ) {materialFunction=&FlattenNormal}.Result
   B=MaterialFunctionCall(
    Normal=TextureSample(UVs=@n111) {samplerType=SAMPLERTYPE_Normal, texture=/Game/Packages/Fortress_Rubble/Textures/T_CliffTile_N}.RGB
    Flatness=ScalarParameter() {defaultValue=0.8728, parameterName=TilingRockNormal}
   ) {materialFunction=&FlattenNormal}.Result
   Alpha=@n118
  )
  StaticSwitchParameter(False=LinearInterpolate(
    B=MaterialFunctionCall(BaseNormal=@n114.RGB, AdditionalNormal=@n125) {materialFunction=&BlendAngleCorrectedNormals}.Result
    Alpha=@n7
   )) {defaultValue=true, parameterName=UseOcean}
  n126 = ScalarParameter() {defaultValue=0.1, parameterName=S Ocean}
  n127 = ComponentMask(None=@n122) {r=true}
  n128 = LinearInterpolate(A=ScalarParameter() {parameterName=S Land}, B=ScalarParameter() {defaultValue=4.6086, parameterName=S Land_2}, Alpha=@n127)
  n129 = LinearInterpolate(A=ScalarParameter() {parameterName=S Land_1}, B=ScalarParameter() {defaultValue=0.0484, parameterName=S Land_3}, Alpha=@n127)
  n130 = Add(A=@n99, B=@n7)
  StaticSwitchParameter(
   True=@n126
   False=LinearInterpolate(A=@n126, B=ShadingPathSwitch(Default=@n128, Mobile=@n129) {desc=Migrated from FeatureLevelSwitch}, Alpha=Clamp(None=@n130))
  ) {defaultValue=true, parameterName=UseOcean}
  n131 = ScalarParameter() {defaultValue=0.2, parameterName=R Ocean_1}
  StaticSwitchParameter(True=@n131, False=LinearInterpolate(A=@n131, B=ScalarParameter() {defaultValue=0.8302, parameterName=R Land}, Alpha=@n7)) {defaultValue=true, parameterName=UseOcean}
  n132 = LinearInterpolate(
   A=TextureSample(UVs=@n103.PS) {texture=/Game/Environments/World/Backgrounds/Locales/Temperate/Textures/T_BackgroundCivilisation_E}.RGB
   B=Constant()
   Alpha=@n108.PS
  )
  n133 = ScalarParameter() {defaultValue=6.9164, parameterName=LightsEmissive}
  StaticSwitchParameter(False=Add(A=Multiply(A=@n132, B=@n133))) {defaultValue=true, parameterName=UseOcean}
  Time()
  VertexInterpolator(VS=LinearInterpolate(A=@n64, Alpha=@n62.G))
  VertexInterpolator(VS=Divide(A=ComponentMask(None=@n4.XYZ) {g=true, r=true}, B=@n105))
  WorldPosition()
 ]
}
