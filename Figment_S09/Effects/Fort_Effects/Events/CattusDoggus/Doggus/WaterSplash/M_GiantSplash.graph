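# Masked, two-sided mesh-particle material for a giant water splash in the CattusDoggus event effects.
# Vertex offsets bend, lift and push the strands, randomized by vertex color and driven by dynamic parameters and a particle-age curve, with SplineThicken setting width.
# Panning noise and caustics, eroded by particle alpha, feed a curve-atlas color ramp and contrast-adjusted, temporally dithered masking.
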Material /Figment_S09/Effects/Fort_Effects/Events/CattusDoggus/Doggus/WaterSplash/M_GiantSplash
graph {
 properties {bUsedWithMeshParticles=true, blendMode=BLEND_Masked, pixelDepthOffsetMode=PDOM_Legacy, twoSided=true, usedShadingModels=Default Lit}
 nodes [
  n1 = DynamicParameter() {defaultValue=<rgba 1 1024 -1 1>, paramNames [BendScale,Width,Z_MoveUp,Time]}
  n2 = TextureCoordinate()
  n3 = Multiply(A=@n1.Width, B=LinearInterpolate(Alpha=ComponentMask(None=@n2) {g=true}) {constA=8, constB=2})
  n4 = OneMinus(None=ComponentMask(None=TextureCoordinate()) {g=true})
  n5 = Constant3Vector() {constant=<rgba 0 0 -10000 0>}
  n6 = VertexColor()
  n7 = LinearInterpolate(B=ScalarParameter() {defaultValue=1.5, parameterName=Z_Bend_RandomLow}, Alpha=Add(A=@n6.R, B=@n6.R)) {constA=1}
  n8 = Multiply(A=Multiply(A=@n4, B=@n4), B=Multiply(A=Multiply(A=@n5, B=@n7), B=@n1.BendScale))
  n9 = Constant3Vector() {constant=<rgba 0 0 10000 0>}
  n10 = LinearInterpolate(A=ScalarParameter() {defaultValue=0.5, parameterName=Z_Up_RandomLow}, Alpha=@n6.B)
  n11 = Multiply(A=@n4, B=Multiply(A=Multiply(A=@n9, B=@n10), B=@n1.Z_MoveUp))
  n12 = ScalarParameter() {defaultValue=0.95, parameterName=X_Offset_RandomLow}
  n13 = Transform(None=Multiply(A=Constant3Vector() {constant=<rgba 0 -500 0 0>}, B=LinearInterpolate(A=@n12, Alpha=@n6.R)))
  n14 = CurveAtlasRowParameter(CurveTime=ParticleRelativeTime()) {
   atlas=&AthenaCurveAtlas01
   curve=/Figment_S09/Athena/Prototype/Blueprints/Cattus_9_4/FX/Curve_Ramps2
   defaultValue=0.4473
   parameterName=Param_2
  }
  n15 = ComponentMask(None=TextureCoordinate()) {g=true}
  n16 = LinearInterpolate(A=@n14.G, B=@n14.G, Alpha=@n15)
  n17 = Add(
   A=Add(A=Add(A=@n8, B=@n11), B=Multiply(A=@n13, B=@n16))
   B=Multiply(A=OneMinus(None=@n15), B=VectorParameter() {parameterName=AxisBiseWS, defaultValue=<rgba 0 -3000 0 1>}.RGB)
  )
  n18 = ComponentMask(None=@n6) {g=true, r=true}
  n19 = VectorParameter() {parameterName=UVS, defaultValue=<rgba 0.3 0.1 0.05 0.25>}
  n20 = Panner(Coordinate=Multiply(A=TextureCoordinate(), B=AppendVector(A=@n19.R, B=@n19.R)), Time=@n1.Time, Speed=AppendVector(A=@n19.B, B=@n19.B))
  n21 = MaterialFunctionCall(
   WidthBase=@n3
   WidthTip=@n3
   WorldPosition=Add(A=@n17, B=WorldPosition() {worldPositionShaderOffset=WPT_ExcludeAllShaderOffsets}.XYZ)
   AdditionalNormal=TextureSample(UVs=Add(A=@n18, B=@n20)) {samplerType=SAMPLERTYPE_Normal, texture=/Game/Effects/Fort_Effects/Textures/WaterFoam0001_19_N}.RGB
  ) {materialFunction=&SplineThicken}
  Add(A=@n17, B=@n21.WorldPositionOffset)
  n22 = TextureCoordinate()
  n23 = VectorParameter() {parameterName=UVS_D2, defaultValue=<rgba 0.7 0.3 -0.5 0.5>}
  n24 = LinearInterpolate(Alpha=ComponentMask(None=@n6) {r=true}) {constA=0.2}
  n25 = Multiply(A=Multiply(A=@n22, B=AppendVector(A=@n23.R, B=@n23.R)), B=@n24)
  n26 = DynamicParameter() {defaultValue=<rgba 1 1024 -1 1>, paramNames [BendScale,Width,Z_MoveUp,Time]}
  n27 = Panner(Coordinate=@n25, Time=@n26.Time, Speed=AppendVector(A=@n23.B, B=@n23.B))
  n28 = ScalarParameter() {defaultValue=512, parameterName=ScreenTexSize}
  n29 = Add(
   A=Multiply(
    A=MaterialFunctionCall!(TextureResolution=@n28) {materialFunction=/Engine/Functions/Engine_MaterialFunctions02/Texturing/ScreenAlignedPixelToPixelUVs}.Result
    B=Constant2Vector() {g=1, r=0.5}
   )
   B=Multiply(A=Time(), B=Constant2Vector() {g=0.2})
  )
  n30 = Multiply(A=ComponentMask(None=TextureSample(UVs=@n29) {samplerType=SAMPLERTYPE_Normal, texture=/Game/Packages/Fortnite_UniqueMaterials_1/Textures/T_WaterSoft_N}.RGB) {g=true, r=true}) {constB=0.1}
  Add(A=@n27, B=@n30)
  n31 = Add(A=@n22, B=ComponentMask(None=VertexColor()) {g=true, r=true})
  n32 = VectorParameter() {parameterName=UVS_D, defaultValue=<rgba 1.5 0.3 0.05 0.25>}
  n33 = AppendVector(A=@n32.R, B=@n32.R)
  n34 = Panner(Coordinate=Multiply(A=Multiply(A=@n31, B=@n33), B=@n24), Time=@n26.Time, Speed=AppendVector(A=@n32.B, B=@n32.B))
  Add(A=@n34, B=@n30)
  n35 = TextureSample(UVs=@n27) {texture=/Figment_S09/Effects/Fort_Effects/Textures/Water/T_Caustics_Softer}
  n36 = MaterialFunctionCall(Float2=Sine(None=TextureCoordinate() {uTiling=0.5, vTiling=0.5})) {materialFunction=&BreakOutFloat2Components}
  n37 = Multiply(
   A=Multiply(A=Multiply(A=@n36.R, B=@n36.R)) {constB=1.2}
   B=TextureSample(UVs=@n34) {texture=/Game/Effects/Fort_Effects/Gameplay/SmokeDustExplosions/Textures/T_NoisePack_01}.RGB
  ) {constB=2}
  n38 = ParticleColor()
  n39 = MaterialFunctionCall(Float3=MaterialFunctionCall!(Blend=LinearInterpolate(A=@n35.R, B=@n36.G) {constAlpha=0.1}, Base=Subtract(A=@n37, B=@n38.A)) {materialFunction=/Engine/Functions/Engine_MaterialFunctions03/Blends/Blend_Screen}.Result) {materialFunction=&BreakOutFloat3Components}
  n40 = CurveAtlasRowParameter(CurveTime=@n39.G) {
   atlas=&AthenaCurveAtlas01
   curve=/Figment_S09/Athena/Prototype/Blueprints/Cattus_9_4/FX/Curve_Color2
   defaultValue=0.4434
   parameterName=Param_1
  }
  n41 = Constant() {r=0.5}
  Add(
   A=Multiply(A=Multiply(A=@n40, B=@n38.RGB), B=ScalarParameter() {defaultValue=0.2, parameterName=EmissiveBost}) {constB=0.2}
   B=Multiply(A=MaterialFunctionCall(In=@n39.G, Contrast=@n41) {materialFunction=&CheapContrast}.Result) {constB=0}
  )
  Constant() {r=0.5}
  Constant() {r=0.2}
  n42 = MaterialFunctionCall(In=@n39.G, Contrast=ScalarParameter() {defaultValue=2, parameterName=AlphaContrast}) {materialFunction=&CheapContrast}
  DepthFade(Opacity=@n42.Result, FadeDistance=ScalarParameter() {defaultValue=2048, parameterName=DF_FD})
  MaterialFunctionCall(Alpha Threshold=@n42.Result) {materialFunction=&DitherTemporalAA}
  Multiply(
   A=Add(A=MaterialFunctionCall() {materialFunction=&DitherTemporalAA_Fortnite}.Result) {constB=0.5}
   B=ScalarParameter() {defaultValue=2048, parameterName=PDO}
  )
  Multiply()
  Multiply(A=@n13, B=ScalarParameter() {defaultValue=600}) {constB=10}
  n43 = Multiply(A=@n25) {constB=4}
  n44 = Multiply(
   A=TextureSample(UVs=Panner(Coordinate=@n43) {speedX=1, speedY=1}) {texture=/Figment_S09/Missions/Secondary/Outpost/IronCity/Pods/Textures/T_Ironcity_Pods_Dots}.R
   B=@n35.R
  )
  Multiply(A=Multiply(A=@n44, B=@n44), B=ScalarParameter() {defaultValue=2048, parameterName=Flash})
  Reroute(None=@n21.Normal)
  Reroute()
  ScalarParameter() {defaultValue=1024, parameterName=Width}
  TextureSample() {texture=&T_NoisePack_01}
  TextureSample() {texture=/Game/Effects/Fort_Effects/Textures/Water/T_MaskeSplashes_Bubbles1_A}
 ]
}
